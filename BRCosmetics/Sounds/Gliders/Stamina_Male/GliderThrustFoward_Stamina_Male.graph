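# Sound cue for the Stamina_Male glider's forward-thrust sound in Battle Royale cosmetics.
# A mixer layers two pitched-down branches: one randomly picks from three thrust waves, the other from two thrust-start waves.
# Playback runs at half volume with glider open/close attenuation and concurrency, the Fort_Gliders_Land sound class, and a glider-loop audio properties sheet.
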SoundCue /BRCosmetics/Sounds/Gliders/Stamina_Male/GliderThrustFoward_Stamina_Male
root=Mixer_0
nodes [
 {id=Mixer_0, children [Modulator_0,Modulator_1], properties {inputVolume [1,1]}}
 {id=Modulator_0, children [Random_0], properties {pitchMax=0.75, pitchMin=0.7}}
 {id=Modulator_1, children [Random_1], properties {pitchMax=0.75, pitchMin=0.7}}
 {id=Random_0, children [WavePlayer_1,WavePlayer_2,WavePlayer_3], properties {weights [1,1,1]}}
 {id=Random_1, children [WavePlayer_4,WavePlayer_5], properties {weights [1,1]}}
 {id=WavePlayer_1, properties {soundWaveAssetPtr=&/Stamina_Glider_Thrust_01}}
 {id=WavePlayer_2, properties {soundWaveAssetPtr=&/Stamina_Glider_Thrust_02}}
 {id=WavePlayer_3, properties {soundWaveAssetPtr=&/Stamina_Glider_Thrust_03}}
 {id=WavePlayer_4, properties {soundWaveAssetPtr=&/Stamina_Glider_ThrustStart_01}}
 {id=WavePlayer_5, properties {soundWaveAssetPtr=&/Stamina_Glider_ThrustStart_02}}
]
properties {
 attenuationSettings=/Game/Sounds/Attenuation/Fort_Player_Glider_OpenClose_Athena_Att
 audioPropertiesSheet=&APS_Cosmetic_Glider_Loop
 maxDistance=10000
 soundClassObject=&Fort_Gliders_Land
 volumeMultiplier=0.5
 concurrencySet [&Athena_GliderOpenClose_Conc]
}
