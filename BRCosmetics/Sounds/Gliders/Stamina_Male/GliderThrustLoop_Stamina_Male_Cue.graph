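# Looping thrust sound cue for the Stamina Male cosmetic glider, using glider thrust attenuation and glider sound class.
# A mixer blends two looping waves: an "Idol" layer at higher input volume, started after a short random delay, and a quieter wind loop.
# Each layer has a continuous modulator that maps the AnyDirection parameter to volume and pitch.
# The Idol layer is pitched higher and keeps a volume floor.
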SoundCue /BRCosmetics/Sounds/Gliders/Stamina_Male/GliderThrustLoop_Stamina_Male_Cue
root=Mixer_1
nodes [
 {id=Delay_1, children [WavePlayer_0], properties {delayMax=0.6, delayMin=0.5}}
 {id=Mixer_1, children [ModulatorContinuous_3, ModulatorContinuous_1], properties {inputVolume [0.8,0.3]}}
 {
  id=ModulatorContinuous_1
  children [WavePlayer_2]
  properties {
   pitchModulationParams {default=0, maxOutput=1.1, minOutput=0.9, parameterName=AnyDirection}
   volumeModulationParams {default=0, maxInput=1.5, parameterName=AnyDirection}
  }
 }
 {
  id=ModulatorContinuous_3
  children [Delay_1]
  properties {
   pitchModulationParams {default=0, maxOutput=1.5, minOutput=1.3, parameterName=AnyDirection}
   volumeModulationParams {default=0, maxInput=1.5, minOutput=0.3, parameterName=AnyDirection}
  }
 }
 {id=WavePlayer_0, properties {bLooping=true, soundWaveAssetPtr=&/Stamina_Glider_Idol}}
 {id=WavePlayer_2, properties {bLooping=true, soundWaveAssetPtr=&/Stamina_Glider_WindLoop}}
]
properties {
 attenuationSettings=&Fort_Player_Glider_Thrust_Athena_Att
 audioPropertiesSheet=&APS_Cosmetic_Glider_Loop
 maxDistance=12500
 soundClassObject=&Fort_Gliders_Land
 volumeMultiplier=0.85
 concurrencySet [null]
}
