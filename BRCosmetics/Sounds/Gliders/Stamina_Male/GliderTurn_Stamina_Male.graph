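# SoundCue for the turn or banking sound of the Stamina Male glider in battle royale.
# It picks at random among four equally weighted glider bank waves, plays them through a modulator that lowers pitch within a narrow range, then a mixer at half overall volume.
# It uses glider open/close attenuation and concurrency, and the glider landing sound class.
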SoundCue /BRCosmetics/Sounds/Gliders/Stamina_Male/GliderTurn_Stamina_Male
root=Mixer_0
nodes [
 {id=Mixer_0, children [Modulator_0,null], properties {inputVolume [1,1]}}
 {id=Modulator_0, children [Random_0], properties {pitchMax=0.75, pitchMin=0.7}}
 {id=Random_0, children [WavePlayer_0,WavePlayer_6,WavePlayer_7,WavePlayer_8], properties {weights [1,1,1,1]}}
 {id=WavePlayer_0, properties {soundWaveAssetPtr=&/Stamina_Glider_Bank_01}}
 {id=WavePlayer_6, properties {soundWaveAssetPtr=&/Stamina_Glider_Bank_02}}
 {id=WavePlayer_7, properties {soundWaveAssetPtr=&/Stamina_Glider_Bank_03}}
 {id=WavePlayer_8, properties {soundWaveAssetPtr=&/Stamina_Glider_Bank_04}}
]
properties {
 attenuationSettings=/Game/Sounds/Attenuation/Fort_Player_Glider_OpenClose_Athena_Att
 maxDistance=10000
 soundClassObject=&Fort_Gliders_Land
 volumeMultiplier=0.5
 concurrencySet [&Athena_GliderOpenClose_Conc]
}
